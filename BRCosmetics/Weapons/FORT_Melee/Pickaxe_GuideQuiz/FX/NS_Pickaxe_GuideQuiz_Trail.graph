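# Niagara system for the red swing trails on the GuideQuiz pickaxe, drawn as anim-trail ribbons.
# Red_Toon_Trail spawns two bright trails and Soft_Trail adds a fainter one, while Red_Toon_TrailRight is disabled.
# A system LRBias bool switches bone or socket indices and swaps the toon trails' lifetimes.
# System alpha fade, dissolve and time-of-day color multipliers scale each ribbon's color.
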NiagaraSystem /BRCosmetics/Weapons/FORT_Melee/Pickaxe_GuideQuiz/FX/NS_Pickaxe_GuideQuiz_Trail
dependencies {
 usedDataInterfaces [
  &NiagaraDataInterfaceSocketReader
  &NiagaraDataInterfaceColorCurve
  &NiagaraDataInterfaceCurve
  /Script/Niagara.NiagaraDataInterfaceArrayBool
 ]
 usedDynamicInputs [
  &ColorFromCurve
  &MergeHelper_FloatSwitch
  &FloatFromCurve
  /Niagara/DynamicInputs/Arrays/SelectBoolFromArray
 ]
 usedModules [
  &self:SystemSpawnScript.SystemScriptSource.SystemScriptGraph.NiagaraNodeAssignment_1.SetVariables_#1
  &SystemState
  &self:SystemSpawnScript.SystemScriptSource.SystemScriptGraph.NiagaraNodeAssignment_3.SetVariables_#2
  &self:SystemSpawnScript.SystemScriptSource.SystemScriptGraph.NiagaraNodeAssignment_0.SetVariables_#3
  &self:SystemSpawnScript.SystemScriptSource.SystemScriptGraph.NiagaraNodeAssignment_4.SetVariables_#4
 ]
}
emitters {
 Red_Toon_Trail {
  emitterUpdateScript [
   &EmitterState()
   &FN_SpawnAnimTrail(Number of Anim Trails=2)
  ]
  particleSpawnScript [
   &FN_InitializeAnimTrailBoneIndecies(First Bone or Socket Index:NiagaraInt32=&MakeCustomIntFromBool(Bool=$System.LRBias, True INT=4, False INT=0))
   &FN_InitializeAnimTrailBoneIndecies(
    AnimTrail Selection=1
    First Bone or Socket Index:NiagaraInt32=&MakeCustomIntFromBool(Bool=$System.LRBias, True INT=2, False INT=5)
    Second Bone or Socket Index=3
   ) {module=FN_InitializeAnimTrailBoneIndecies001}
   &FN_InitializeAnimTrail(
    Color=&MakeLinearColorFromVectorAndFloat(
     "Vector (RGB)"=&MakeVectorFromLinearColorRGB(Color=<rgba 1 0 0 1>)
     "Float (Alpha)"=&Multiply_Float(A=$System.AlphaFade, B=0.95)
    )
    Lifetime=&MaskFloat(
     Mask Source=0
     Mask Parameter=?
     Mask=0
     Masked Float=&MakeCustomFloatFromBool(Bool=$System.LRBias, True Float=0.12, False Float=0.18)
     Passthrough Float=&MakeCustomFloatFromBool(Bool=$System.LRBias, True Float=0.18, False Float=0.12)
    )
   )
  ]
  particleUpdateScript [
   &FN_UpdateAnimTrail() {module=FN_UpdateAnimTrail001}
   &ScaleColor(
    Scale Alpha=$System.AlphaDissolve
    Scale RGB=&MakeVectorFromLinearColorRGB(Color=$System.TODColorMultiplier)
    ScaleA=true
    ScaleRGB=true
   )
  ]
  rendererClasses [&NiagaraRibbonRendererProperties]
  renderers [{rendererClass=&NiagaraRibbonRendererProperties}]
 }
 Red_Toon_TrailRight {
  bEnabled=false
  emitterUpdateScript [
   &EmitterState()
   &FN_SpawnAnimTrail()
  ]
  particleSpawnScript [
   &FN_InitializeAnimTrailBoneIndecies(
    First Bone or Socket Index:NiagaraInt32=&MakeCustomIntFromBool(Bool=$System.LRBias, True INT=2, False INT=5)
    Second Bone or Socket Index=3
   )
   &FN_InitializeAnimTrail(
    Color=&MakeLinearColorFromVectorAndFloat(
     "Vector (RGB)"=&MakeVectorFromLinearColorRGB(Color=<rgba 1 0 0 1>)
     "Float (Alpha)"=&Multiply_Float(A=$System.AlphaFade, B=0.95)
    )
    Lifetime=&MakeCustomFloatFromBool(Bool=$System.LRBias, True Float=0.175, False Float=0.0875)
   )
  ]
  particleUpdateScript [
   &FN_UpdateAnimTrail() {module=FN_UpdateAnimTrail001}
   &ScaleColor(
    Scale Alpha=$System.AlphaDissolve
    Scale RGB=&MakeVectorFromLinearColorRGB(Color=$System.TODColorMultiplier)
    ScaleA=true
    ScaleRGB=true
   )
  ]
  rendererClasses [&NiagaraRibbonRendererProperties]
  renderers [{rendererClass=&NiagaraRibbonRendererProperties}]
 }
 Soft_Trail {
  emitterUpdateScript [
   &EmitterState()
   &FN_SpawnAnimTrail()
  ]
  particleSpawnScript [
   &FN_InitializeAnimTrailBoneIndecies(
    First Bone or Socket Index:NiagaraInt32=&MakeCustomIntFromBool(Bool=$System.LRBias, True INT=2, False INT=0)
    Second Bone or Socket Index:NiagaraInt32=&MakeCustomIntFromBool(Bool=$System.LRBias, True INT=3, False INT=1)
   )
   &FN_InitializeAnimTrail(
    Color=&MakeLinearColorFromVectorAndFloat(
     "Vector (RGB)"=&MakeVectorFromLinearColorRGB(Color=<rgba 1 0.05 0.05 1>)
     "Float (Alpha)"=&Multiply_Float(A=$System.AlphaFade, B=0.25)
    )
    Lifetime=0.175
   )
  ]
  particleUpdateScript [
   &FN_UpdateAnimTrail() {module=FN_UpdateAnimTrail001}
   &ScaleColor(
    Scale Alpha=$System.AlphaDissolve
    Scale RGB=&MakeVectorFromLinearColorRGB(Color=$System.TODColorMultiplier)
    ScaleA=true
    ScaleRGB=true
   )
  ]
  rendererClasses [&NiagaraRibbonRendererProperties]
  renderers [{rendererClass=&NiagaraRibbonRendererProperties}]
 }
}
summary {
 systemName=NS_Pickaxe_GuideQuiz_Trail
 emitters [
  {emitterName=Soft_Trail, rendererClasses [&NiagaraRibbonRendererProperties]}
  {bEnabled=false, emitterName=Red_Toon_TrailRight, rendererClasses [&NiagaraRibbonRendererProperties]}
  {emitterName=Red_Toon_Trail, rendererClasses [&NiagaraRibbonRendererProperties]}
 ]
 userVariables [
  User.RibbonIDOffset {type=NiagaraInt32}
  User.SwingSectionIndex {type=NiagaraInt32}
 ]
}
